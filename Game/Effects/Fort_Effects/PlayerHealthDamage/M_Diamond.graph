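# M_Diamond is an unlit translucent particle material in the player health damage effects, using a diamond texture for opacity.
# Particle color, optionally hue-shifted by a static switch, feeds the shared emissive adjustments function.
# Particle alpha times a selectable texture channel gives opacity, multiplied by a depth fade.
# On low quality, that opacity is multiplied by itself instead of the depth fade.
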Material /Game/Effects/Fort_Effects/PlayerHealthDamage/M_Diamond
graph {
 properties {
  bUsedWithNanite=true
  bUsedWithNiagaraMeshParticles=true
  bUsedWithNiagaraSprites=true
  bUsedWithParticleSprites=true
  blendMode=BLEND_Translucent
  pixelDepthOffsetMode=PDOM_Legacy
  shadingModel=MSM_Unlit
  usedShadingModels=Unlit
 }
 nodes [
  n1 = ParticleColor()
  MaterialFunctionCall(Emissive=StaticSwitchParameter(
    True=MaterialFunctionCall(Hue Shift Percentage=@n1.RGB) {materialFunction=&HueShift}.Result
    False=@n1.RGB
   ) {parameterName=UseHueShift?}) {materialFunction=&MF_FX_EmissiveAdjustments}
  n2 = Multiply(
   A=@n1.A
   B=ChannelMaskParameter(None=TextureSampleParameter2D() {parameterName=Opacity Texture, texture=&/T_Diamond}.RGBA) {parameterName=Opacity Texture Channel}.R
  )
  n3 = ScalarParameter() {defaultValue=12, parameterName=FadeDistance}
  Multiply(A=@n2, B=QualitySwitch(Default=DepthFade(FadeDistance=@n3), Low=@n2))
 ]
}
