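# Unlit translucent sprite material for player health damage particles, usable with Cascade and Niagara, that paints a tinted glow near scene surfaces.
# Emissive adds a time-driven sine shimmer from macro-UV noise to an EdgeGlow color scaled by Fresnel.
# The Fresnel reads flattened scene-texture world normals, with constant substitutes on forward and mobile paths.
# Opacity combines particle alpha, Fresnel, a radial mask and paired depth fades that confine visibility to a soft band near geometry.
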Material /Game/Effects/Fort_Effects/PlayerHealthDamage/M_Health_Particles
graph {
 properties {
  bUsedWithNiagaraSprites=true
  bUsedWithParticleSprites=true
  blendMode=BLEND_Translucent
  pixelDepthOffsetMode=PDOM_Legacy
  shadingModel=MSM_Unlit
  usedShadingModels=Unlit
 }
 nodes [
  n1 = ScalarParameter() {defaultValue=6, parameterName=MacroUV Scale}
  n2 = Multiply(
   A=ParticleMacroUV()
   B=ShadingPathSwitch(Default=@n1, Deferred=@n1, Mobile=ScalarParameter() {defaultValue=2, parameterName=MacroUV Scale Mobile})
  ) {constB=5}
  n3 = Add(
   A=OneMinus(None=TextureSampleParameter2D(UVs=@n2) {parameterName=MacroTexture, texture=&T_NoisePack_01}.R)
   B=Multiply(A=Time()) {constB=2}
  )
  n4 = Multiply(
   A=Add(A=MaterialFunctionCall!(Sine Phase=@n3, Value 1, Value 2) {materialFunction=/Engine/Functions/Engine_MaterialFunctions02/Utility/Sine_Remapped}.Remapped Sine) {constB=0.05}
   B=ScalarParameter() {defaultValue=0.75, parameterName=Macro Intensity}
  )
  n5 = ParticleColor()
  n6 = Multiply(A=VectorParameter() {parameterName=EdgeGlow, defaultValue=<rgba 0.0426 2 0.6853 1>}.RGB, B=@n5.RGB)
  n7 = SceneTexture() {sceneTextureId=PPI_WorldNormal}
  n8 = Fresnel(
   ExponentIn=DynamicParameter() {paramNames [FresnelExponent, Refraction, GlowPan, Param4]}.FresnelExponent
   Normal=MaterialFunctionCall(Normal=ComponentMask(None=@n7.Color) {b=true, g=true, r=true}, Flatness=Constant() {r=0.2}) {materialFunction=&FlattenNormal}.Result
  ) {baseReflectFraction=0, exponent=1}
  n9 = ShadingPathSwitch(Default=@n8, Deferred=@n8, Forward=Constant3Vector() {constant=<rgba 0 0 1 0>}, Mobile=@n8)
  n10 = ShadingPathSwitch(Default=@n9, Deferred=@n9, Mobile=Constant() {r=0.1}) {desc=Migrated from FeatureLevelSwitch}
  MaterialFunctionCall(Emissive=Add(A=Multiply(A=@n4, B=@n5.RGB), B=Multiply(A=@n6, B=@n10))) {materialFunction=&MF_FX_EmissiveAdjustments}
  MaterialFunctionCall() {materialFunction=&DebugTimeSine}
  n11 = TextureCoordinate()
  n12 = Constant() {r=0.5}
  n13 = DepthFade(FadeDistance=ScalarParameter() {defaultValue=35, parameterName=VisibleDepthDistance}) {fadeDistanceDefault=80}
  Multiply(
   A=Multiply(A=Multiply(A=@n5.A, B=@n10), B=SphereMask(A=@n11, B=@n12, Radius=@n12) {hardnessPercent=0})
   B=DepthFade(Opacity=OneMinus(None=@n13), FadeDistance=ScalarParameter() {defaultValue=15, parameterName=DepthFade}) {fadeDistanceDefault=15}
  )
 ]
}
