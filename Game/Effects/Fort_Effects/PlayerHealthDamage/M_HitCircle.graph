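# Translucent unlit sprite material for a hit circle in player health damage effects, colored by particle color.
# Opacity subtracts an inner, offsettable sphere mask from an outer one, forming a ring or crescent whose clipping is driven by dynamic parameters, fully or partly via a static switch.
# An optional switch erodes the shape with tiled noise.
# Particle alpha and depth fade, skipped on low quality, scale the result.
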Material /Game/Effects/Fort_Effects/PlayerHealthDamage/M_HitCircle
graph {
 properties {
  bUsedWithNanite=true
  bUsedWithNiagaraSprites=true
  bUsedWithParticleSprites=true
  blendMode=BLEND_Translucent
  pixelDepthOffsetMode=PDOM_Legacy
  shadingModel=MSM_Unlit
  usedShadingModels=Unlit
 }
 nodes [
  n1 = ParticleColor()
  MaterialFunctionCall(Emissive=@n1.RGB) {materialFunction=&MF_FX_EmissiveAdjustments}
  n2 = TextureCoordinate()
  n3 = DynamicParameter() {defaultValue=<rgba 1 1 0.55 1>, paramNames [ClipAwayCircle, Outer Edge Hardness, Inner Offset Amount, Clip Away Outer Circle]}
  n4 = Constant() {r=0.5}
  n5 = Constant() {r=1}
  n6 = Constant() {r=0.55}
  n7 = Subtract(
   A=SphereMask(
    A=@n2
    B=Constant() {r=0.5}
    Radius=StaticSwitchParameter(True=@n3.Clip Away Outer Circle, False=@n4) {parameterName=Use Full Dynamic Parameter Control}
    Hardness=StaticSwitchParameter(True=@n3.Outer Edge Hardness, False=@n5) {parameterName=Use Full Dynamic Parameter Control}
   )
   B=SphereMask(
    A=@n2
    B=StaticSwitchParameter(True=@n3.Inner Offset Amount, False=@n6) {parameterName=Use Full Dynamic Parameter Control}
    Radius=@n3.ClipAwayCircle
    Hardness=Constant() {r=0.6}
   )
  )
  n8 = Multiply(A=TextureCoordinate(), B=ScalarParameter() {defaultValue=0.75, parameterName=Tiling})
  n9 = ScalarParameter() {defaultValue=2, parameterName=Mip Level}
  n10 = Subtract(
   A=@n7
   B=Power(
    Base=TextureSample!(UVs=@n8, Tex, Level=@n9, Apply View MipBias) {mipValueMode=TMVM_MipLevel, samplerType=SAMPLERTYPE_LinearColor, texture=/Game/Global/Textures/Default/Noise/T_Noise01_Inverse}.G
    Exp=ScalarParameter() {defaultValue=4, parameterName=Noise Exp}
   )
  )
  n11 = DepthFade(FadeDistance=Constant() {r=12})
  Multiply(
   A=Multiply(A=@n1.A, B=StaticSwitchParameter(True=@n10, False=@n7) {parameterName=Breakup Circle Shape})
   B=QualitySwitch(Default=@n11, Low=Constant() {r=1}, High=@n11, Medium=@n11)
  )
 ]
}
