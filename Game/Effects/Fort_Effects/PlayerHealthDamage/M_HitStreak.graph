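# M_HitStreak is an unlit, translucent, two-sided particle sprite material for hit streaks in the player health damage effects.
# Particle color RGB feeds the MF_FX_EmissiveAdjustments function for the glow color.
# Particle alpha multiplies the T_HitStreak texture's RGB to form a streak mask.
# A short depth fade multiplies that mask to soften the sprite where it meets scene geometry.
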Material /Game/Effects/Fort_Effects/PlayerHealthDamage/M_HitStreak
graph {
 properties {
  bUsedWithParticleSprites=true
  blendMode=BLEND_Translucent
  pixelDepthOffsetMode=PDOM_Legacy
  shadingModel=MSM_Unlit
  twoSided=true
  usedShadingModels=Unlit
 }
 nodes [
  n1 = ParticleColor()
  MaterialFunctionCall(Emissive=@n1.RGB) {materialFunction=&MF_FX_EmissiveAdjustments}
  Multiply(
   A=Multiply(A=@n1.A, B=TextureSample() {texture=&/T_HitStreak}.RGB)
   B=DepthFade(FadeDistance=Constant() {r=12})
  )
 ]
}
