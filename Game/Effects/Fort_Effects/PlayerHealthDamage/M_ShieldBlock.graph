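# Translucent, unlit, two-sided mesh-particle material for a player shield-block effect, tinted by particle color.
# Opacity sums a clipped radial mask with world-height bands, panning caustic clouds and a thin ring, softened by inverted Fresnel and depth fade.
# A time-based sine along vertex normals adds jitter.
# Dynamic parameters drive clip radius, circle radius, jitter and cloud strength.
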Material /Game/Effects/Fort_Effects/PlayerHealthDamage/M_ShieldBlock
graph {
 properties {
  bUsedWithMeshParticles=true
  bUsedWithNanite=true
  bUsedWithNiagaraMeshParticles=true
  blendMode=BLEND_Translucent
  pixelDepthOffsetMode=PDOM_Legacy
  shadingModel=MSM_Unlit
  twoSided=true
  usedShadingModels=Unlit
 }
 nodes [
  n1 = ParticleColor()
  MaterialFunctionCall(Emissive=@n1.RGB) {materialFunction=&MF_FX_EmissiveAdjustments}
  n2 = Sine(None=Multiply(A=Time()) {constB=10})
  Multiply(
   A=Multiply(A=VertexNormalWS(), B=Multiply(A=@n2) {constB=2})
   B=DynamicParameter() {defaultValue=<rgba 0 1 0 1>, paramNames [ClipRadius, CircleRadius, JitterStrength, CloudStrength]}.JitterStrength
  )
  n3 = SphereMask(A=TextureCoordinate(), B=Constant() {r=0.5}, Radius=Constant() {r=0.5}, Hardness=Constant())
  n4 = TextureCoordinate()
  n5 = Constant() {r=0.5}
  n6 = DynamicParameter() {defaultValue=<rgba 0 1 0 1>, paramNames [ClipRadius, CircleRadius, JitterStrength, CloudStrength]}
  n7 = Constant() {r=0.1}
  n8 = Multiply(A=@n1.A, B=Subtract(A=Power(Base=@n3), B=SphereMask(A=@n4, B=@n5, Radius=@n6.ClipRadius, Hardness=@n7)))
  n9 = ComponentMask(None=WorldPosition().XYZ) {b=true}
  n10 = Add(A=Frac(None=Multiply(A=@n9) {constB=0.3})) {constB=0.2}
  n11 = Clamp(None=Add(A=Floor(None=@n10)) {constB=0.8})
  n12 = TextureCoordinate() {coordinateIndex=1}
  n13 = TextureCoordinate()
  n14 = Constant() {r=0.5}
  n15 = Constant() {r=0.5}
  n16 = TextureCoordinate()
  n17 = Constant() {r=0.5}
  n18 = Constant() {r=0.5}
  n19 = Constant() {r=0.1}
  n20 = Multiply(
   A=TextureSample(UVs=Panner(Coordinate=@n12) {speedY=-1.5}) {texture=&T_ShieldCaustics}.R
   B=Subtract(A=SphereMask(A=@n13, B=@n14, Radius=@n14, Hardness=@n15), B=SphereMask(A=@n16, B=@n17, Radius=@n18, Hardness=@n19))
  )
  n21 = DynamicParameter() {defaultValue=<rgba 0 1 0 1>, paramNames [ClipRadius, CircleRadius, JitterStrength, CloudStrength]}
  n22 = DynamicParameter() {defaultValue=<rgba 0 1 0 1>, paramNames [ClipRadius, CircleRadius, JitterStrength, CloudStrength]}
  n23 = SphereMask(A=TextureCoordinate(), B=Constant() {r=0.5}, Radius=@n22.CircleRadius, Hardness=Constant() {r=0.95})
  n24 = SphereMask(A=TextureCoordinate(), B=Constant() {r=0.5}, Radius=Add(A=@n22.CircleRadius) {constB=-0.2}, Hardness=Constant() {r=0.85})
  n25 = Add(A=Add(A=Multiply(A=@n8, B=@n11), B=Multiply(A=@n20, B=@n21.CloudStrength) {constB=0.5}), B=Multiply(A=Subtract(A=@n23, B=@n24)) {constB=0.1})
  n26 = LinearInterpolate(Alpha=MaterialFunctionCall(Exponent=Constant() {r=4}, Vertex / Pixel Normal=PixelNormalWS()) {materialFunction=&MF_Fresnel_FX}.Fresnel Inverted) {constA=0.05}
  n27 = Constant() {r=20}
  Reroute(None=Multiply(A=Multiply(A=@n25, B=@n26), B=DepthFade(FadeDistance=@n27)))
 ]
}
